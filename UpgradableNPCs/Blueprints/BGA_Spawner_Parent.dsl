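; Parent interactable spawner actor that, on first use, spawns an AI through the Living World system and converts it to the interacting pawn.
; At BeginPlay it registers as a density reservoir with the Living World Manager once playlist data is ready.
; Interaction runs once, releasing the reservoir, enabling its static point provider and requesting event generation.
; The provider's spawn callback triggers conversion, while an AISpawned flag blocks further interaction.
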(event EventBeginPlay
  (bind _returnvalue (Class|FortAthenaLivingWorldManager|GetLivingWorldManager))
  ;[1 1 1] Init and Registed AI density to Living World Manager
  (Utilities|FlowControl|Sequence
    (:then_0
      (Networking|CanExecuteCosmeticEvents
        (:False
          (Variables|InternalData|SetAISpawned false))))
    (:then_1
      (Variables|Interaction|SetbAllowInteract true)
      (Class|FortAsyncActionCurrentPlaylistReady|PlaylistDataReadyAsync
        (:OnReady
          (Utilities|FlowControl|Delay 0.2)
          (Utilities|IsValid _returnvalue
            (:Is%20Valid
              (Class|FortAthenaLivingWorldManager|RegisterDensityReservoir _returnvalue self 1.0 1.2e+04))))))))

(event Interaction|EventBlueprintOnInteract
  (params (InteractingPawn :type Fort%20Pawn%20Object%20Reference) (InteractionBeingAttempted :type EInteractionBeingAttempted%20Enum :default "FirstInteraction"))
  (bind _returnvalue (Class|FortAthenaLivingWorldManager|GetLivingWorldManager))
  (bind _staticpointprovider_spawnlocation (Variables|Default|GetStaticPointProvider_SpawnLocation))
  ;[1 1 1] When Interaction Is Completed
  (Utilities|FlowControl|DoOnce false
    (:Completed
      (Variables|InternalData|SetInteracting%20Pawn InteractingPawn)
      (Class|FortAthenaLivingWorldManager|UnregisterDensityReservoir _returnvalue self)
      (Utilities|IsValid _staticpointprovider_spawnlocation
        (:Is%20Valid
          (Class|FortAthenaLivingWorldPointProviderInterface|EnablePointProvider_DEPRECATED _staticpointprovider_spawnlocation)
          (EventDispatchers|BindEventtoOnActorSpawned _staticpointprovider_spawnlocation (EventDispatchers|CreateEvent :!event Custom|AISpawnedFromInteraction))
          (Class|FortAthenaLivingWorldManager|RequestEventGeneration _returnvalue)
          (Variables|InternalData|SetAISpawned true))))))

(event Custom|AISpawnedFromInteraction
  (params (SpawnedActor :type Actor%20Object%20Reference))
  ;[1 1 1] Make the AI Converted with the one interacting with the Pawn
  (bind _asfort_pawn (Utilities|Casting|CastToFortPawn SpawnedActor))
  (Convert|TriggerConversion (Variables|InternalData|GetInteracting%20Pawn) _asfort_pawn))

(fn BlueprintCanInteract
  (params (InteractingPawn :type Fort%20Pawn%20Object%20Reference) (InteractionBeingAttempted :type EInteractionBeingAttempted%20Enum :default "FirstInteraction") (InteractionType :type TInteractionType%20Enum :default "IT_NoInteraction"))
  (return (not (Variables|InternalData|GetAISpawned))))

(fn BlueprintGetInteractionString
  (params (InteractingPawn :type Fort%20Pawn%20Object%20Reference) (InteractionBeingAttempted :type EInteractionBeingAttempted%20Enum :default "FirstInteraction"))
  (return (Variables|GetBuildingActorInteractionText)))

(fn ConstructionScript)
